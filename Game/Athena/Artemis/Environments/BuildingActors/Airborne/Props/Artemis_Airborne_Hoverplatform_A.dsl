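; Airborne hoverplatform prop Blueprint whose construction script varies the look of its balloon flame mesh per instance.
; It runs the parent construction script, then destroys the point light component.
; It then sets random scalar material parameters on the flame mesh for hue shift, vertical and small movement strength, and opacity.
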(fn ConstructionScript
  (bind _balloonflame_protomesh (Variables|Default|GetBalloonFlame_protoMesh))
  (|CallParentFunction)
  (Components|DestroyComponent (Variables|Default|GetPointLight))
  (Rendering|Material|SetScalarParameterValueonMaterials _balloonflame_protomesh "HueShiftPercent" (Math|Random|RandomFloatinRange 0.45 0.6))
  (Rendering|Material|SetScalarParameterValueonMaterials _balloonflame_protomesh "UpDownStrength" (Math|Random|RandomFloatinRange 2.0 25.0))
  (Rendering|Material|SetScalarParameterValueonMaterials _balloonflame_protomesh "SmallMoveStrength" (Math|Random|RandomFloatinRange 0.0 8.0))
  (Rendering|Material|SetScalarParameterValueonMaterials _balloonflame_protomesh "OpacityMult" (Math|Random|RandomFloatinRange 0.1 1.0)))
