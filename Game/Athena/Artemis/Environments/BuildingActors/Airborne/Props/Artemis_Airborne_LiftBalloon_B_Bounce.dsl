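; Bouncy lift balloon prop (variant B) for the Artemis Airborne environment that bounces actors touching its jump collision.
; On Jump_Collision overlap, an actor with a valid ability system component receives the BounceGE gameplay effect, and a quest stat event tagged Jump.OnUmbrella is sent.
; On death, it fades out the balloon loop sound once both audio components validate; the burner loop is checked but never faded.
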(event EventOnDeathPlayEffects
  (params (Damage :type Float%20%28single-precision%29) (DamageTags :type Gameplay%20Tag%20Container%20Structure%20%28by%20ref%29) (Momentum :type Vector :default "0, 0, 0") (HitInfo :type Hit%20Result%20Structure%20%28by%20ref%29) (InstigatedBy :type Fort%20Pawn%20Object%20Reference) (DamageCauser :type Actor%20Object%20Reference) (EffectContext :type Gameplay%20Effect%20Context%20Handle%20Structure))
  (bind _spot_balloon_liftballoon_loop_cue (Variables|Default|GetSpot_Balloon_LiftBalloon_Loop_Cue))
  (bind _spot_balloon_liftballoon_burner_loop_cue (Variables|Default|GetSpot_Balloon_LiftBalloon_Burner_Loop_Cue))
  (Utilities|IsValid _spot_balloon_liftballoon_loop_cue
    (:Is%20Valid
      (Utilities|IsValid _spot_balloon_liftballoon_burner_loop_cue
        (:Is%20Valid
          (if (and (Utilities|IsValid _spot_balloon_liftballoon_burner_loop_cue :!class /Script/Engine.KismetSystemLibrary) (Utilities|IsValid _spot_balloon_liftballoon_loop_cue :!class /Script/Engine.KismetSystemLibrary))
            (Audio|Components|Audio|FadeOut _spot_balloon_liftballoon_loop_cue 0.2 0.0 "Linear" :!class /Script/Engine.AudioComponent)))))))

(event EventBeginPlay)

(event Collision|EventActorBeginOverlap
  (params (OtherActor :type Actor%20Object%20Reference)))

(event EventTick
  (params (DeltaSeconds :type Float%20%28single-precision%29)))

(event OnComponentBeginOverlap(Jump_Collision)
  (params (OverlappedComponent :type Primitive%20Component%20Object%20Reference) (OtherActor :type Actor%20Object%20Reference) (OtherComp :type Primitive%20Component%20Object%20Reference) (OtherBodyIndex :type Integer) (bFromSweep :type Boolean) (SweepResult :type Hit%20Result%20Structure%20%28by%20ref%29))
  (bind _returnvalue (Ability|GetAbilitySystemComponent OtherActor))
  (Utilities|IsValid _returnvalue
    (:Is%20Valid
      (GameplayEffects|ApplyGameplayEffectToSelf :self _returnvalue :GameplayEffectClass (Variables|Default|GetBounceGE) :Level 0.0)
      (Quest|SendComplexCustomStatEvent (Quests|GetQuestManager (Fortnite|GetFortPlayerControllerfromActor OtherActor) "Athena") 0 "(GameplayTags=((TagName=\"Athena.Quests.Jump.OnUmbrella\")))" "(GameplayTags=)" 1))))

(fn ConstructionScript
  (|CallParentFunction))
