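; Airborne balloon launch pad prop that plays a bounce burst effect when something overlaps its launchpad collision.
; The overlap event executes a burst GameplayCue on itself, using the tag from GetGCNTag_BounceBurst.
; The cue parameters carry the pad's location, the overlapping actor as instigator and the pad as effect causer.
; No launch logic appears here, and the construction script only calls the parent.
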(event OnComponentBeginOverlap(S_Athena_Launchpad_Collision)
  (params (OverlappedComponent :type Primitive%20Component%20Object%20Reference) (OtherActor :type Actor%20Object%20Reference) (OtherComp :type Primitive%20Component%20Object%20Reference) (OtherBodyIndex :type Integer) (bFromSweep :type Boolean) (SweepResult :type Hit%20Result%20Structure%20%28by%20ref%29))
  (GameplayCue|ExecuteGameplayCueOnActor(Burst) self (Variables|Components|GetGCNTag_BounceBurst) (Ability|GameplayCue|MakeGameplayCueParameters :NormalizedMagnitude 0.0 :RawMagnitude 0.0 :Location (Transformation|GetActorLocation) :Normal "0,0,0" :Instigator OtherActor :EffectCauser self :GameplayEffectLevel 1 :AbilityLevel 1 :bReplicateLocationWhenUsingMinimalRepProxy false)))

(fn ConstructionScript
  (|CallParentFunction))
